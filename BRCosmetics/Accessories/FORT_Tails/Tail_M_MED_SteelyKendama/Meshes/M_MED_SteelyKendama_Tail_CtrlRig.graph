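# Control Rig for the SteelyKendama tail cosmetic, posing the tail and correcting its helper bones.
# It first bends the tail from joint 5 onward, fading the pose as Velocity rises via interpolated remapping scaled by tailPoseAlpha.
# It then spreads plus and minus half of each joint's relative rotation onto helper bones and parent-constrains ring03 across weighted tail joints.
# Time-offset lag and base-bend transform nodes exist but are not wired to execution.
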ControlRigBlueprint /BRCosmetics/Accessories/FORT_Tails/Tail_M_MED_SteelyKendama/Meshes/M_MED_SteelyKendama_Tail_CtrlRig
graphs [
 New Function_ContainedGraph {nodes [
   n1 = RigVMFunctionEntryNode()
   Cast = /Script/RigVM.RigVMFunction_MathQuaternionToRotator(Value=&RigUnit_GetRelativeTransformForItem(bChildInitial=False, bParentInitial=False, Child=@n1.Child, Parent=@n1.Parent).RelativeTransform.Rotation) {is=struct RigVMUnitNode, template="Cast::Execute(in Value,out Result)"}
   n2 = RigVMDispatchNode(Struct=@Cast.Result) {template="DISPATCH_RigVMDispatch_BreakStruct(in Struct,out Elements)"}
   n3 = &RigVMFunction_MathDoubleMul(A=@n2.Elements.Roll, B=@n1.Factor)
   n4 = &RigVMFunction_MathDoubleMul(A=@n2.Elements.Pitch, B=@n1.Factor)
   n5 = &RigVMFunction_MathDoubleMul(A=@n2.Elements.Yaw, B=@n1.Factor)
   > RigVMFunctionReturnNode(Result=/Script/RigVM.RigVMFunction_MathQuaternionFromRotatorV2(Value=RigVMDispatchNode(Elements=(Pitch=0,Yaw=0,Roll=0), Elements.Roll=@n3.Result, Elements.Pitch=@n4.Result, Elements.Yaw=@n5.Result) {template="DISPATCH_RigVMDispatch_MakeStruct(in Elements,out Struct)"}.Struct) {is=struct RigVMUnitNode, template="Cast::Execute(in Value,out Result)"}.Result)
  ]}
 RigVMFunctionLibrary {nodes [Get Relative Rotation mult by Factor = RigVMCollapseNode(Child=(Type=Bone,Name="C_Tail_A_5_Jnt"), Factor=-0.5, Parent=(Type=Bone,Name="C_Tail_A_4_Jnt"))]}
 RigVMFunction_Sequence_1_ContainedGraph {nodes [
   n1 = RigVMFunctionEntryNode()
   RigVMFunction_Sequence = &RigVMFunction_Sequence(ExecuteContext=@n1.ExecuteContext)
   n2 = &RigVMFunction_Sequence(ExecuteContext=@RigVMFunction_Sequence.B)
   RigVMFunctionReturnNode(A=@RigVMFunction_Sequence.A, B=@n2.A, C=@n2.B)
  ]}
 RigVMModel {nodes [
   RigUnit_BeginExecution = &RigUnit_BeginExecution()
   RigVMFunction_Sequence = RigVMAggregateNode(ExecuteContext=@RigUnit_BeginExecution.ExecutePin)
   &RigUnit_ParentConstraint(
    AdvancedSettings=(InterpolationType=Average,RotationOrderForFilter=XZY)
    Child=(Type=Bone,Name="dyn_ring03_root")
    Parents="((Item=(Type=Bone,Name=\"C_Tail_A_8_Jnt\"),Weight=0.1306),(Item=(Type=Bone,Name=\"C_Tail_A_6_Jnt\"),Weight=0.0315),(Item=(Type=Bone,Name=\"C_Tail_A_5_Jnt\"),Weight=0.0243),(Item=(Type=Bone,Name=\"C_Tail_A_7_Jnt\"),Weight=0.211),(Item=(Type=Bone,Name=\"C_Tail_A_4_Jn ...(truncated)"
    Weight=1
    ExecutePin=@RigVMFunction_Sequence.C
   )
   n1 = &RigVMFunction_MathQuaternionFromRotatorV2(Value=RigVMDispatchNode(Elements=(Pitch=60,Yaw=0,Roll=0)) {template="DISPATCH_RigVMDispatch_MakeStruct(in Elements,out Struct)"}.Struct) {template="Cast::Execute(in Value,out Result)"}
   AlphaInterp = &RigVMFunction_AlphaInterp(
    InRange=(Min=0,Max=770)
    InterpSpeedDecreasing=15
    InterpSpeedIncreasing=15
    OutRange=(Min=1,Max=0)
    Scale=1
    bClampResult=true
    bMapRange=true
    Value=RigVMVariableNode() {variable=Velocity}.Value
   )
   Multiply = &RigVMFunction_MathDoubleMul(A=@AlphaInterp.Result, B=RigVMVariableNode() {variable=tailPoseAlpha}.Value)
   &RigUnit_ModifyTransforms(
    ItemToModify="((Item=(Type=Bone,Name=\"C_Tail_A_5_Jnt\"),Transform=(Rotation=(X=0,Y=-0.2588,Z=0,W=0.9659),Translation=(X=0,Y=0,Z=0),Scale3D=(X=1,Y=1,Z=1))),(Item=(Type=Bone,Name=\"C_Tail_A_6_Jnt\"),Transform=(Rotation=(X=0,Y=-0.5,Z=0,W=0.866),Translation=(X=0,Y=0,Z=0),Scale ...(truncated)"
    WeightMinimum=0
    ItemToModify.1.Transform.Rotation=@n1.Result
    ItemToModify.2.Transform.Rotation=@n1.Result
    ItemToModify.3.Transform.Rotation=@n1.Result
    Weight=@Multiply.Result
    ExecutePin=@RigVMFunction_Sequence.A
   )
   Get Relative Rotation mult by Factor = RigVMFunctionReferenceNode(Child=(Type=Bone,Name="C_Tail_A_5_Jnt"), Factor=-0.5, Parent=(Type=Bone,Name="C_Tail_A_4_Jnt"))
   Get Relative Rotation mult by Factor_1 = RigVMFunctionReferenceNode(Child=(Type=Bone,Name="C_Tail_A_5_Jnt"), Factor=0.5, Parent=(Type=Bone,Name="C_Tail_A_4_Jnt"))
   Get Relative Rotation mult by Factor_2 = RigVMFunctionReferenceNode(Child=(Type=Bone,Name="C_Tail_A_6_Jnt"), Factor=-0.5, Parent=(Type=Bone,Name="C_Tail_A_5_Jnt"))
   Get Relative Rotation mult by Factor_3 = RigVMFunctionReferenceNode(Child=(Type=Bone,Name="C_Tail_A_6_Jnt"), Factor=0.5, Parent=(Type=Bone,Name="C_Tail_A_5_Jnt"))
   Get Relative Rotation mult by Factor_4 = RigVMFunctionReferenceNode(Child=(Type=Bone,Name="C_Tail_A_7_Jnt"), Factor=-0.5, Parent=(Type=Bone,Name="C_Tail_A_6_Jnt"))
   Get Relative Rotation mult by Factor_5 = RigVMFunctionReferenceNode(Child=(Type=Bone,Name="C_Tail_A_7_Jnt"), Factor=0.5, Parent=(Type=Bone,Name="C_Tail_A_6_Jnt"))
   Get Relative Rotation mult by Factor_6 = RigVMFunctionReferenceNode(Child=(Type=Bone,Name="C_Tail_A_8_Jnt"), Factor=-0.5, Parent=(Type=Bone,Name="C_Tail_A_7_Jnt"))
   Get Relative Rotation mult by Factor_6_1 = RigVMFunctionReferenceNode(Child=(Type=Bone,Name="C_Tail_A_8_Jnt"), Factor=0.5, Parent=(Type=Bone,Name="C_Tail_A_7_Jnt"))
   &RigUnit_ModifyTransforms(
    ItemToModify="((Item=(Type=Bone,Name=\"dyn_Tail_A_5_Jnt_add\"),Transform=(Rotation=(X=0,Y=0,Z=0,W=1),Translation=(X=0,Y=0,Z=0),Scale3D=(X=1,Y=1,Z=1))),(Item=(Type=Bone,Name=\"dyn_Tail_A_5_Jnt_add02\"),Transform=(Rotation=(X=0,Y=0,Z=0,W=1),Translation=(X=0,Y=0,Z=0),Scale3D=( ...(truncated)"
    Weight=1
    WeightMinimum=0
    ItemToModify.0.Transform.Rotation=@"Get Relative Rotation mult by Factor".Result
    ItemToModify.1.Transform.Rotation=@"Get Relative Rotation mult by Factor_1".Result
    ItemToModify.2.Transform.Rotation=@"Get Relative Rotation mult by Factor_2".Result
    ItemToModify.3.Transform.Rotation=@"Get Relative Rotation mult by Factor_3".Result
    ItemToModify.4.Transform.Rotation=@"Get Relative Rotation mult by Factor_4".Result
    ItemToModify.5.Transform.Rotation=@"Get Relative Rotation mult by Factor_5".Result
    ItemToModify.6.Transform.Rotation=@"Get Relative Rotation mult by Factor_6".Result
    ItemToModify.7.Transform.Rotation=@"Get Relative Rotation mult by Factor_6_1".Result
    ExecutePin=@RigVMFunction_Sequence.B
   )
   RigVMFunction_MathQuaternionToEuler = &RigVMFunction_MathQuaternionToEuler(
    RotationOrder=ZYX
    Value=&RigUnit_GetRelativeTransformForItem(Child=(Type=Bone,Name="C_Tail_A_6_Jnt"), Parent=(Type=Bone,Name="C_Tail_A_6_Jnt"), bChildInitial=False, bParentInitial=true).RelativeTransform.Rotation
   )
   n2 = /Script/RigVM.RigVMFunction_MathVectorMul(
    B=(X=-0.5,Y=1,Z=-1)
    A=/Script/RigVM.RigVMFunction_TimeOffsetVector(BufferSize=140, SecondsAgo=0.9, TimeRange=2, Value=@RigVMFunction_MathQuaternionToEuler.Result) {is=struct RigVMUnitNode, template="TimeOffset::Execute(in Value,in SecondsAgo,in BufferSize,in TimeRange,out Result)"}.Result
   ) {is=struct RigVMUnitNode, template="Multiply::Execute(in A,in B,out Result)"}
   RigVMFunction_MathQuaternionFromRotatorV2 = &RigVMFunction_MathQuaternionFromRotatorV2(Value=RigVMDispatchNode(Elements=(Pitch=0,Yaw=0,Roll=0), Elements.Roll=@n2.Result.X, Elements.Yaw=@n2.Result.Z) {template="DISPATCH_RigVMDispatch_MakeStruct(in Elements,out Struct)"}.Struct) {template="Cast::Execute(in Value,out Result)"}
   TimeOffset = &RigVMFunction_TimeOffsetVector(
    BufferSize=140
    SecondsAgo=0.9
    TimeRange=2
    Value=&RigVMFunction_MathQuaternionToEuler(
     RotationOrder=ZYX
     Value=&RigUnit_GetRelativeTransformForItem(Child=(Type=Bone,Name="C_Tail_A_7_Jnt"), Parent=(Type=Bone,Name="C_Tail_A_7_Jnt"), bChildInitial=False, bParentInitial=true).RelativeTransform.Rotation
    ).Result
   ) {template="TimeOffset::Execute(in Value,in SecondsAgo,in BufferSize,in TimeRange,out Result)"}
   n3 = &RigVMFunction_MathVectorMul(B=(X=-0.5,Y=1,Z=-1.5), A=@TimeOffset.Result) {template="Multiply::Execute(in A,in B,out Result)"}
   &RigUnit_ModifyTransforms(
    ExecutePin=()
    ItemToModify="((Item=(Type=Bone,Name=\"C_Tail_A_6_Jnt\"),Transform=(Rotation=(X=0,Y=0,Z=0,W=1),Translation=(X=0,Y=0,Z=0),Scale3D=(X=1,Y=1,Z=1))),(Item=(Type=Bone,Name=\"C_Tail_A_7_Jnt\"),Transform=(Rotation=(X=0,Y=0,Z=0,W=1),Translation=(X=0,Y=0,Z=0),Scale3D=(X=1,Y=1,Z=1))) ...(truncated)"
    WeightMinimum=0
    ItemToModify.0.Transform.Rotation=@RigVMFunction_MathQuaternionFromRotatorV2.Result
    ItemToModify.1.Transform.Rotation=&RigVMFunction_MathQuaternionFromRotatorV2(Value=RigVMDispatchNode(Elements=(Pitch=0,Yaw=0,Roll=0), Elements.Roll=@n3.Result.X, Elements.Yaw=@n3.Result.Z) {template="DISPATCH_RigVMDispatch_MakeStruct(in Elements,out Struct)"}.Struct) {template="Cast::Execute(in Value,out Result)"}.Result
    Weight=@Multiply.Result
   )
   &RigUnit_ModifyTransforms(
    ExecutePin=()
    ItemToModify="((Item=(Type=Bone,Name=\"C_Tail_A_1_Jnt\"),Transform=(Rotation=(X=0,Y=0,Z=0.1224,W=0.9925),Translation=(X=0,Y=0,Z=0),Scale3D=(X=1,Y=1,Z=1))),(Item=(Type=Bone,Name=\"C_Tail_A_2_Jnt\"),Transform=(Rotation=(X=0,Y=0,Z=0.1224,W=0.9925),Translation=(X=0,Y=0,Z=0),Sca ...(truncated)"
    Weight=1
    WeightMinimum=0
   )
  ]}
]
hierarchy [
 C_Root_Main_Root_Jnt {loc=<0 0 0>, rot=<0 0 0 1>}
 C_Tail_A_Base_Jnt {parent=C_Root_Main_Root_Jnt, loc=<1.4997 -0.3348 0.4223>, rot=<-0.5 0.5 0.5 -0.5>}
 C_Tail_A_1_Jnt {parent=C_Tail_A_Base_Jnt, loc=<8.8758 -0.0429 0.1469>, rot=<0.7071 0.0063 0.0115 -0.707>}
 C_Tail_A_2_Jnt {parent=C_Tail_A_1_Jnt, loc=<8.8775 0.0376 0.011>, rot=<0 0.0009 -0.0031 -1>}
 C_Tail_A_3_Jnt {parent=C_Tail_A_2_Jnt, loc=<8.879 0.0015 0.0004>, rot=<0 0 -0.0001 -1>}
 C_Tail_A_4_Jnt {parent=C_Tail_A_3_Jnt, loc=<8.8778 0 0>, rot=<0 0 0 1>}
 C_Tail_A_5_Jnt {parent=C_Tail_A_4_Jnt, loc=<8.8778 0 0>, rot=<0 0 0 1>}
 C_Tail_A_6_Jnt {parent=C_Tail_A_5_Jnt, loc=<8.8775 0 0>, rot=<0 0 0 1>}
 C_Tail_A_7_Jnt {parent=C_Tail_A_6_Jnt, loc=<8.8776 0 0>, rot=<0 0 0 1>}
 C_Tail_A_8_Jnt {parent=C_Tail_A_7_Jnt, loc=<8.8788 0 0>, rot=<0 0 0 1>}
 dyn_Tail_A_8_Jnt_add01 {parent=C_Tail_A_8_Jnt, loc=<0 0 0>, rot=<0 0 0 1>}
 dyn_Tail_A_8_Jnt_add02 {parent=dyn_Tail_A_8_Jnt_add01, loc=<4.438 0 0>, rot=<0 0 0 1>}
 C_Tail_A_9_Jnt {parent=C_Tail_A_8_Jnt, loc=<8.8781 0 0>, rot=<0 0 0 1>}
 dyn_Tail_A_7_Jnt_add01 {parent=C_Tail_A_7_Jnt, loc=<0 0 0>, rot=<0 0 0 1>}
 dyn_Tail_A_7_Jnt_add02 {parent=dyn_Tail_A_7_Jnt_add01, loc=<4.438 0 0>, rot=<0 0 0 1>}
 dyn_ring03_root {parent=dyn_Tail_A_7_Jnt_add01, loc=<0.5099 -0.1026 -0.4723>, rot=<0.506 -0.4939 -0.5033 -0.4967>}
 dyn_ring03_01 {parent=dyn_ring03_root, loc=<0 0 0>, rot=<0 0 0 1>}
 dyn_Tail_A_6_Jnt_add01 {parent=C_Tail_A_6_Jnt, loc=<0 0 0>, rot=<0 0 0 1>}
 dyn_Tail_A_6_Jnt_add02 {parent=dyn_Tail_A_6_Jnt_add01, loc=<4.438 0 0>, rot=<0 0 0 1>}
 dyn_ring02_root {parent=dyn_Tail_A_6_Jnt_add02, loc=<2.165 -0.0562 -0.4876>, rot=<0.506 -0.4939 -0.5033 -0.4967>}
 dyn_ring02_01 {parent=dyn_ring02_root, loc=<0 0 0>, rot=<0 0 0 1>}
 dyn_ring01_root {parent=dyn_Tail_A_6_Jnt_add02, loc=<-0.4707 -0.0071 -0.502>, rot=<0.506 -0.4939 -0.5033 -0.4967>}
 dyn_ring01_01 {parent=dyn_ring01_root, loc=<0 0 0>, rot=<0 0 0 1>}
 dyn_Tail_A_5_Jnt_add {parent=C_Tail_A_5_Jnt, loc=<0 0 0>, rot=<0 0 0 1>}
 dyn_Tail_A_5_Jnt_add02 {parent=dyn_Tail_A_5_Jnt_add, loc=<4.438 0 0>, rot=<0 0 0 1>}
]
variables [
 Velocity {type=double, default=0}
 tailPoseAlpha {type=double, default=1}
]
